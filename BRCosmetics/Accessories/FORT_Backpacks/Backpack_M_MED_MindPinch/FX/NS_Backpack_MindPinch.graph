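# Niagara system for the MindPinch backpack cosmetic, producing a warm yellow-orange blinking glow on the accessory.
# One Airglow emitter burst-spawns particles that follow sampled skeletal mesh locations, drawn through sprite and mesh renderers with fixed sprite facing and camera offset.
# A system-level Waveform drives Blink, which modulates both the sprite alpha and a separate glow color.
# Time-of-day functions tint that glow color and set alpha levels, lowering them at night.
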NiagaraSystem /BRCosmetics/Accessories/FORT_Backpacks/Backpack_M_MED_MindPinch/FX/NS_Backpack_MindPinch
dependencies {
 usedDataInterfaces [&NiagaraDataInterfaceCurve, &NiagaraDataInterfaceSkeletalMesh]
 usedDynamicInputs [&RandomRangeFloat, &FloatFromCurve]
 usedModules [
  &self:SystemSpawnScript.SystemScriptSource.SystemScriptGraph.NiagaraNodeAssignment_0.SetVariables_#1
  &SystemState
  /Niagara/Noise/Waveform
 ]
}
emitters {Airglow {
  emitterUpdateScript [
   &EmitterState()
   &SpawnBurst_Instantaneous(Use Spawn Probability=false, UseLoopCountLimit=false)
  ]
  particleSpawnScript [
   &InitializeParticle(
    Color=<rgba 2 1.706 0.2043 0.3>
    Lifetime=2.3
    Position Offset=<0 -15 -1>
    Sprite Size=<22 12>
    Sprite Size Mode=::NewEnumerator4
    UsePositionOffset=false
    Write Material Random=true
   )
   &SpriteFacingAndAlignment(Sprite Facing=<0 -1 0>, WriteAlignment=false, WriteFacing=true)
   &CameraOffset(Camera Offset Amount=6)
  ]
  particleUpdateScript [
   &ParticleState()
   &SkeletalMeshLocation(
    MaskSpawnGroup=false
    Mesh Sampling Type=::NewEnumerator5
    Sampled Position Offset=<0 0 -0.8>
    Skeletal Mesh=$System.Skeletal Mesh
    Socket Sampling Mode=::NewEnumerator5
   )
   &ScaleSpriteSize(Scale Sprite Size Mode=::NewEnumerator0)
   /SetVariables(Particles.Blink:NiagaraFloat=$System.Waveform.Waveform) {isSetParameters=true}
   /SetVariables(
    Particles.ColorGlow:LinearColor=&MakeLinearColorFromVectorAndFloat(
     "Vector (RGB)"=&MakeVectorFromLinearColorRGB(Color=&FNTimeOfDayColor(Night Color=<rgba 20 10.1051 0 1>, Morning Color=<rgba 20 15.8403 0 1>, Day Color=<rgba 30 23.7604 0 1>, Evening Color=<rgba 20 15.8403 0 1>))
     "Float (Alpha)"=&Multiply_Float(A=&FNTimeOfDayScalar(Night Scale=0.07, Morning Scale=1, Evening Scale=1), B=$Particles.Blink)
    )
    Particles.PositionGlow:NiagaraPosition=&Add_Vector(A=Linked() {linkedVariable {name=Output.SkeletalMeshLocation.SampledPositionWithoutOffset, type=NiagaraPosition}}, B=<0 0.1 -0.1>)
   ) {isSetParameters=true}
   &ScaleColor(
    Scale Alpha=&Multiply_Float(A=&FNTimeOfDayScalar(Night Scale=0.3, Morning Scale=1, Evening Scale=1), B=$Particles.Blink)
    ScaleA=true
    ScaleRGB=false
   )
  ]
  rendererClasses [&NiagaraSpriteRendererProperties, &NiagaraMeshRendererProperties]
  renderers [
   {rendererClass=&NiagaraSpriteRendererProperties}
   {rendererClass=&NiagaraMeshRendererProperties, rendererIndex=1}
  ]
 }}
summary {
 systemName=NS_Backpack_MindPinch
 emitters [{emitterName=Airglow, rendererClasses [&NiagaraSpriteRendererProperties, &NiagaraMeshRendererProperties]}]
 userVariables [
  User.bIsFrontEndPreview {type=NiagaraBool}
  User.CycleFX {type=NiagaraFloat, defaultValue=0, description=""}
  User.ToggleFX {type=NiagaraInt32}
 ]
}
