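# Physics asset for the MindPinch backpack cosmetic, defining simulated capsule bodies for its dangling antenna, arms, fingers, legs and feet.
# A small capsule on the pack bone is the root, with a three-segment antenna chain, both arms and both legs constrained to it.
# Three fingers hang off each arm and a foot off each leg, with tight swing limits throughout and locked twist on legs and feet.
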PhysicsAsset /BRCosmetics/Accessories/FORT_Backpacks/Backpack_M_MED_MindPinch/Meshes/M_MED_MindPinch_Pack_Physics
bodies {
 dyn_antenna_base [{shapeType=Capsule, length=5, radius=3, shapeName=dyn_antenna_base_capsule, center=<2 0 0>, rotation=<rot 90 0 0>}]
 dyn_antenna_mid [{shapeType=Capsule, length=5, radius=3, shapeName=dyn_antenna_mid_capsule, center=<2 0 0>, rotation=<rot 90 0 0>}]
 dyn_antenna_tip [{shapeType=Capsule, length=5, radius=3, shapeName=dyn_antenna_tip_capsule, center=<2 0 0>, rotation=<rot 90 0 0>}]
 dyn_arm_l [{
   shapeType=Capsule
   length=2.7272
   radius=2.6292
   shapeName=dyn_arm_l_capsule
   center=<2.9063 -0.1104 -0.1241>
   rotation=<rot 0.0542 -90 -2.4502>
  }]
 dyn_arm_r [{
   shapeType=Capsule
   length=2.7271
   radius=2.6292
   shapeName=dyn_arm_r_capsule
   center=<-2.9063 0.1104 0.1241>
   rotation=<rot 0.0542 -90 -2.4503>
  }]
 dyn_foot_l [{shapeType=Capsule, length=5, radius=3, shapeName=dyn_foot_l_capsule, center=<2 0 0>, rotation=<rot 90 0 0>}]
 dyn_foot_r [{shapeType=Capsule, length=8, radius=6, shapeName=dyn_foot_r_capsule, center=<-2 0 0>, rotation=<rot 90 0 0>}]
 dyn_index_finger_l [{
   shapeType=Capsule
   length=1.6849
   radius=0.8851
   shapeName=dyn_index_finger_l_capsule
   center=<2.0954 0.3337 0.159>
   rotation=<rot 0 -90.3962 -86.4791>
  }]
 dyn_index_finger_r [{
   shapeType=Capsule
   length=1.2584
   radius=1.2364
   shapeName=dyn_index_finger_r_capsule
   center=<-1.3248 -0.3311 -0.2202>
   rotation=<rot 41.5918 -90 -34.3564>
  }]
 dyn_leg_l [{shapeType=Capsule, length=5, radius=3, shapeName=dyn_leg_l_capsule, center=<2 0 0>, rotation=<rot 90 0 0>}]
 dyn_leg_r [{shapeType=Capsule, length=10, radius=7, shapeName=dyn_leg_r_capsule, center=<-2 0 0>, rotation=<rot 90 0 0>}]
 dyn_middle_finger_l [{
   shapeType=Capsule
   length=1.8488
   radius=0.9796
   shapeName=dyn_middle_finger_l_capsule
   center=<1.8488 -0.6768 -0.1712>
   rotation=<rot 0 -90.835 -86.2017>
  }]
 dyn_middle_finger_r [{
   shapeType=Capsule
   length=1.3103
   radius=1.2986
   shapeName=dyn_middle_finger_r_capsule
   center=<-1.0234 0.6624 0.0957>
   rotation=<rot 0 -74.313 -60.2594>
  }]
 dyn_pinky_finger_l [{
   shapeType=Capsule
   length=1.6862
   radius=0.9008
   shapeName=dyn_pinky_finger_l_capsule
   center=<2.2279 0.2088 -0.5062>
   rotation=<rot 0 -91.9193 -89.1802>
  }]
 dyn_pinky_finger_r [{
   shapeType=Capsule
   length=1.2858
   radius=1.1986
   shapeName=dyn_pinky_finger_r_capsule
   center=<-1.4709 -0.2211 0.4851>
   rotation=<rot 0 58.1595 -61.6343>
  }]
 pack [{shapeType=Capsule, length=0.5, radius=0.5, shapeName=pack_capsule}]
}
constraints [
 {bone1Name=dyn_antenna_base, bone2Name=pack, swing1LimitDegrees=5, swing2LimitDegrees=5, twistLimitDegrees=1}
 {bone1Name=dyn_antenna_mid, bone2Name=dyn_antenna_base, swing1LimitDegrees=5, swing2LimitDegrees=5, twistLimitDegrees=1}
 {bone1Name=dyn_antenna_tip, bone2Name=dyn_antenna_mid, swing1LimitDegrees=5, swing2LimitDegrees=5, twistLimitDegrees=1}
 {bone1Name=dyn_arm_l, bone2Name=pack, swing1LimitDegrees=25, swing2LimitDegrees=25, twistLimitDegrees=1}
 {bone1Name=dyn_index_finger_l, bone2Name=dyn_arm_l, swing1LimitDegrees=20, swing2LimitDegrees=20, twistLimitDegrees=1}
 {bone1Name=dyn_middle_finger_l, bone2Name=dyn_arm_l, swing1LimitDegrees=20, swing2LimitDegrees=20, twistLimitDegrees=1}
 {bone1Name=dyn_pinky_finger_l, bone2Name=dyn_arm_l, swing1LimitDegrees=20, swing2LimitDegrees=20, twistLimitDegrees=1}
 {bone1Name=dyn_leg_l, bone2Name=pack, swing1LimitDegrees=20, swing2LimitDegrees=10, twistLimitDegrees=1, twistMotion=Locked}
 {bone1Name=dyn_foot_l, bone2Name=dyn_leg_l, swing1LimitDegrees=15, swing2LimitDegrees=15, twistLimitDegrees=1, twistMotion=Locked}
 {bone1Name=dyn_arm_r, bone2Name=pack, swing1LimitDegrees=25, swing2LimitDegrees=25, twistLimitDegrees=1}
 {bone1Name=dyn_index_finger_r, bone2Name=dyn_arm_r, swing1LimitDegrees=20, swing2LimitDegrees=20, twistLimitDegrees=1}
 {bone1Name=dyn_middle_finger_r, bone2Name=dyn_arm_r, swing1LimitDegrees=20, swing2LimitDegrees=20, twistLimitDegrees=1}
 {bone1Name=dyn_pinky_finger_r, bone2Name=dyn_arm_r, swing1LimitDegrees=20, swing2LimitDegrees=20, twistLimitDegrees=1}
 {bone1Name=dyn_leg_r, bone2Name=pack, swing1LimitDegrees=20, swing2LimitDegrees=10, twistLimitDegrees=45, twistMotion=Locked}
 {bone1Name=dyn_foot_r, bone2Name=dyn_leg_r, swing1LimitDegrees=15, swing2LimitDegrees=15, twistLimitDegrees=45, twistMotion=Locked}
]
