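# Animation Blueprint for the MindPinch backpack cosmetic, playing an idle loop and a one-shot damage reaction.
# An Idle/Reactive state machine feeds RigidBody physics, with a curve controlling the physics alpha.
# Health or shield-decrease anim input events briefly set TookDamage to trigger the reaction, which also fires on a delay in the front end.
# The blink_pose curve drives two complementary material parameters that swap open and closed eye elements.
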AnimBlueprint /BRCosmetics/Accessories/FORT_Backpacks/Backpack_M_MED_MindPinch/Meshes/M_MED_MindPinch_Pack_AnimBP
target_skeleton=&/M_MED_MindPinch_Pack_Skeleton
graphs [
 AnimGraph {nodes [
   Input Pose()
   AnimGraphNode_StateMachine_0 = Idle_Reactive() {is=title AnimGraphNode_StateMachine}
   n1 = RigidBody!(ComponentPose=Local To Component(LocalPose=@AnimGraphNode_StateMachine_0), AlphaCurveName:Name=disable_dynamics, ExternalForce) -> (Pose) {
    alphaInputType=Curve
    bClampLinearTranslationLimitToRefPose=true
    bOverrideWorldGravity=true
    overridePhysicsAsset=&/M_MED_MindPinch_Pack_Physics
    simulationSpace=BaseBoneSpace
    alphaBoolBlend {blendOption=Linear}
    baseBoneRef {boneName=pack}
    componentLinearAccScale=<0.4 0.2 0.2>
    componentLinearVelScale=<0.22 0.22 0.22>
    overrideWorldGravity {z=-980}
    simSpaceSettings {velocityScaleZ=0, worldAlpha=1}
   }
   Output Pose(Result=Component To Local(ComponentPose=@n1)) {layerGroup=None}
  ]}
 Idle_Reactive {
  entry=Idle
  path=AnimGraph.AnimGraphNode_StateMachine_0.Idle_Reactive
  ids=id
  nodes [
   AnimStateNode_0 = Idle()
   AnimStateNode_1 = Reactive() {is=title AnimStateNode}
   AnimStateTransitionNode_0 = AnimStateTransitionNode() {crossfadeDuration=0.05}
   AnimStateTransitionNode_1 = AnimStateTransitionNode() {bAutomaticRuleBasedOnSequencePlayerInState=true, crossfadeDuration=0.05}
  ]
  links [
   {
    dst=Reactive
    rule=AnimGraph.AnimGraphNode_StateMachine_0.Idle_Reactive.AnimStateTransitionNode_0.Transition
    src=Idle
    via=AnimStateTransitionNode_0
   }
   {
    dst=Idle
    rule=AnimGraph.AnimGraphNode_StateMachine_0.Idle_Reactive.AnimStateTransitionNode_1.Transition
    src=Reactive
    via=AnimStateTransitionNode_1
   }
  ]
 }
 Idle {
  path=AnimGraph.AnimGraphNode_StateMachine_0.Idle_Reactive.AnimStateNode_0.Idle
  nodes [Output Animation Pose(Result=Sequence Player 'MindPinch_Pack_Anim_Idle'() {sequence=/BRCosmetics/Accessories/FORT_Backpacks/Backpack_M_MED_MindPinch/Anim/MindPinch_Pack_Anim_Idle})]
 }
 Reactive {
  path=AnimGraph.AnimGraphNode_StateMachine_0.Idle_Reactive.AnimStateNode_1.Reactive
  nodes [Output Animation Pose(Result=Sequence Player 'MindPinch_Pack_Anim_TakeDamage'() {bLoopAnimation=false, sequence=/BRCosmetics/Accessories/FORT_Backpacks/Backpack_M_MED_MindPinch/Anim/MindPinch_Pack_Anim_TakeDamage})]
 }
 Transition {
  path=AnimGraph.AnimGraphNode_StateMachine_0.Idle_Reactive.AnimStateTransitionNode_0.Transition
  nodes [Result(bCanEnterTransition=Get TookDamage() -> (TookDamage:bool))]
 }
 Transition {path=AnimGraph.AnimGraphNode_StateMachine_0.Idle_Reactive.AnimStateTransitionNode_1.Transition, nodes [Result()]}
 EventGraph {nodes [
   n1 = Cast To FortPlayerPawn(Object=GetOwningActor())
   n2 = GetHealth!(self:Fort Pawn Object Reference=@n1) -> (ReturnValue:float)
   > Set Current Health(Current Health:double=@n2) -> (Output_Get:double)
   > n3 = Branch(Condition="Not Equal (Float)"(A=@n2, B=Get Previous Health() -> (Previous Health:double)))
   > Set TookDamage(TookDamage:bool=true) -> (Output_Get:bool)
   > Set Previous Health(Previous Health:double=Get Current Health() -> (Current Health:double)) -> (Output_Get:double)
   Set TookDamage(execute=@n3.else, TookDamage=false)
   Event BlueprintUpdateAnimation()
   > n4 = Sequence()
   Branch(execute=@n4.then_1, Condition=Get bIsInFrontEnd())
   > n5 = Delay(Duration=6.5)
   Set TookDamage(execute=[@n5.then, @n9.then], TookDamage=true)
   > RetriggerableDelay(Duration=0.2)
   > Set TookDamage(TookDamage=false)
   n6 = Set Eye Closed Pose(execute=@n4.then_0, Eye Closed Pose:double=GetCurveValue(CurveName=blink_pose)) -> (Output_Get:double)
   > Set Eye Open Pose(Eye Open Pose:double="Absolute (Float)"(A=float - float(A=@n6, B=1))) -> (Output_Get:double)
   > n7 = Get Dynamic Material Instance(SkeletalMeshComponent=GetPartSkeletalMeshComponent())
   > SetScalarParameterValue(self=@n7, ParameterName=Hide Element 01, Value=Get Eye Closed Pose() -> (Eye Closed Pose:double))
   > SetScalarParameterValue(self=@n7, ParameterName=Hide Element 02, Value=Get Eye Open Pose() -> (Eye Open Pose:double))
   n8 = Event BlueprintProcessAnimInputEvent()
   > n9 = Branch(Condition=OR Boolean(
     A=IsNamed(self=@n8.AnimInputEvent, InEventName=aie.PlayerVitality.HealthDecreased)
     B=IsNamed(self=@n8.AnimInputEvent, InEventName=aie.PlayerVitality.ShieldsDecreased)
    ))
   Event BlueprintPostEvaluateAnimation() {is=title K2Node_Event}
   > Set TookDamage(TookDamage=false)
  ]}
]
variables [
 Current Health {type=number}
 Eye Closed Pose {type=number}
 Eye Open Pose {type=number, default=1}
 Previous Health {type=number}
 TookDamage {type=boolean}
]
